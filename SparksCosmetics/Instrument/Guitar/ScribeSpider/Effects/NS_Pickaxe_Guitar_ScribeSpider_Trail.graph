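# Niagara ribbon trail for swings of the ScribeSpider guitar pickaxe, built as a single toon-trail emitter.
# Shared anim-trail functions spawn, initialize from bone indices and update short-lived particles, with user parameters for ribbon ID offset and swing section.
# Alpha follows a fade-out curve multiplied by System.AlphaFade, and RGB is scaled by a system time-of-day color multiplier.
# The AddVelocity module is disabled.
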NiagaraSystem /SparksCosmetics/Instrument/Guitar/ScribeSpider/Effects/NS_Pickaxe_Guitar_ScribeSpider_Trail
dependencies {
 usedDataInterfaces [&NiagaraDataInterfaceCurve, &NiagaraDataInterfaceSocketReader, &NiagaraDataInterfaceColorCurve]
 usedDynamicInputs [&ColorFromCurve, &MergeHelper_FloatSwitch]
 usedModules [
  &self:SystemSpawnScript.SystemScriptSource.SystemScriptGraph.NiagaraNodeAssignment_5.SetVariables_#1
  &SystemState
  &self:SystemSpawnScript.SystemScriptSource.SystemScriptGraph.NiagaraNodeAssignment_3.SetVariables_#2
  &self:SystemSpawnScript.SystemScriptSource.SystemScriptGraph.NiagaraNodeAssignment_1.SetVariables_#3
 ]
}
emitters {Black_Toon_Trai {
  emitterSpawnScript [&FN_RibbonIDOffsetUtility()]
  emitterUpdateScript [
   &EmitterState()
   &FN_SpawnAnimTrail()
  ]
  particleSpawnScript [
   &FN_InitializeAnimTrailBoneIndecies()
   &FN_InitializeAnimTrail(Lifetime=0.16)
   &AddVelocity(Velocity=<500 0 0>) {enabled=false}
  ]
  particleUpdateScript [
   &SolveForcesAndVelocity(Clamp Velocity=false, Limit Acceleration=false)
   &FN_UpdateAnimTrail() {module=FN_UpdateAnimTrail001}
   &ScaleColor(
    Scale Alpha=&Multiply_Float(
     A=&FloatFromCurve(FloatCurve=DataInterface() {propertyValues {Curve {keys [<at 0.1 1 RCIM_Cubic>, {arriveTangent=-0.6389, tangentMode=RCTM_User, time=0.8, value=0}]}}})
     B=$System.AlphaFade
    )
    Scale RGB=&MakeVectorFromLinearColorRGB(Color=$System.TODColorMultiplier)
    ScaleA=true
    ScaleRGB=true
   )
  ]
  rendererClasses [&NiagaraRibbonRendererProperties]
  renderers [{rendererClass=&NiagaraRibbonRendererProperties}]
 }}
summary {
 systemName=NS_Pickaxe_Guitar_ScribeSpider_Trail
 emitters [{emitterName=Black_Toon_Trai, rendererClasses [&NiagaraRibbonRendererProperties]}]
 userVariables [
  User.RibbonIDOffset {type=NiagaraInt32}
  User.SwingSectionIndex {type=NiagaraInt32}
  User.TrailColor {type=LinearColor, description="", defaultValue {struct=LinearColor, value=<rgba 1 1 1 0.3333>}}
 ]
}
